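# Material function for the BasketSuburb pickaxe that generates 2D coordinates on a virtual, object-aligned plane at an adjustable depth offset from a pivot.
# It intersects the camera ray with that plane, using pixel depth to rescale the camera-to-surface vector.
# The pivot-relative hit is projected onto tangent and binormal axes, which default to instance axes.
# It outputs scaled, 0.5-centered UVs and raw coordinates.
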MaterialFunction /BRCosmetics/Weapons/FORT_Melee/Pickaxe_BasketSuburb/Materials/MF_ObjectAlignedVirtualPlaneCoordinates
graph {
 properties {bExposeToLibrary=true, description=Draws 2d coordinate at a specified alignment and position., libraryCategoriesText [Coordinates]}
 nodes [
  n1 = Constant3Vector() {constant=<rgba 0 1 0 0>}
  n2 = FunctionInput(Preview=Normalize(VectorInput=Transform(None=@n1) {transformSourceType=TRANSFORMSOURCE_Instance})) {bUsePreviewValueAsDefault=true, inputName=World Space Projection Normal, sortPriority=2, previewValue=<1 0 0 1>}
  n3 = CameraPositionWS() {desc=Lo}
  n4 = FunctionInput(Preview=ActorPositionWS()) {bUsePreviewValueAsDefault=true, inputName=Projection Pivot Position, sortPriority=-1}
  n5 = FunctionInput() {bUsePreviewValueAsDefault=true, inputName=Plane Depth Offset, inputType=FunctionInput_Scalar, previewValue=<10 0 0 1>}
  n6 = Multiply(A=Add(A=DotProduct(A=@n4, B=@n2), B=@n5), B=@n2) {desc=Po}
  n7 = Divide(A=DotProduct(A=@n2, B=Subtract(A=@n3, B=@n6) {desc=Lo . Po}) {desc=. N}, B=DotProduct(A=@n2, B=CameraVectorWS() {desc=D}) {desc=l . n}) {desc="Distance to Plane (Varying Camera)"}
  n8 = DotProduct(A=Transform(None=Constant3Vector() {constant=<rgba 0 0 -1 0>}) {transformSourceType=TRANSFORMSOURCE_View}, B=CameraVectorWS() {desc=D}) {desc=l . n}
  n9 = Divide(A=Multiply(A=Multiply(A=@n7, B=@n8)), B=PixelDepth().R)
  n10 = CameraPositionWS()
  n11 = Subtract(A=WorldPosition().XYZ, B=@n10)
  n12 = Subtract(A=Add(A=Multiply(A=@n9, B=@n11), B=@n10), B=@n4)
  n13 = Constant3Vector() {constant=<rgba 0 0 1 0>}
  n14 = FunctionInput(Preview=Normalize(VectorInput=Transform(None=@n13) {transformSourceType=TRANSFORMSOURCE_Instance})) {bUsePreviewValueAsDefault=true, inputName=World Space Projection Tangent, sortPriority=3, previewValue=<0 1 0 1>}
  n15 = Constant3Vector() {constant=<rgba 1 0 0 0>}
  n16 = FunctionInput(Preview=Normalize(VectorInput=Transform(None=@n15) {transformSourceType=TRANSFORMSOURCE_Instance})) {bUsePreviewValueAsDefault=true, inputName=World Space Projection Binormal, sortPriority=4, previewValue=<0 0 1 1>}
  n17 = AppendVector(A=DotProduct(A=@n12, B=@n14), B=DotProduct(A=@n12, B=@n16))
  n18 = FunctionInput() {bUsePreviewValueAsDefault=true, inputName="Coordinate Scale (WS)", inputType=FunctionInput_Scalar, sortPriority=1, previewValue=<10 0 0 1>}
  FunctionOutput(None=Add(A=Divide(A=@n17, B=@n18) {constB=20}) {constB=0.5}) {outputName=Center Aligned UV Plane}
  FunctionOutput(None=@n17) {outputName=Non Biased Coordinates}
  Multiply(A=MaterialFunctionCall() {materialFunction=/Engine/Functions/Engine_MaterialFunctions02/WorldPositionOffset/ObjectScale}.Scale X) {constB=10}
 ]
}
