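# Material function for the BasketSuburb pickaxe that blends an animated, see-into face into the emissive of incoming material attributes.
# Face UVs project local position along the camera ray for fake depth, then scale, rotate and shift with view direction as pseudo-reflection.
# These UVs drive a time-looped two-frame flipbook, blended by its alpha, Fresnel-based opacity and a rectangular mask trimmed by a rerouted mask.
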MaterialFunction /BRCosmetics/Weapons/FORT_Melee/Pickaxe_BasketSuburb/Materials/MF_Pickaxe_BasketSuburb
graph {nodes [
  n1 = GetMaterialAttributes(None=FunctionInput() {inputName=Result, inputType=FunctionInput_MaterialAttributes, sortPriority=10}) {attributeGetTypes [#1, #2]}
  n2 = Time()
  n3 = ScalarParameter() {defaultValue=0.15, parameterName=Fade Idle Speed}
  n4 = VertexInterpolator(VS=LocalPosition() {localOrigin=InstancePreSkinning}.XYZ)
  n5 = TransformPosition(None=CameraPositionWS()) {transformSourceType=TRANSFORMPOSSOURCE_World}
  n6 = Multiply(A=Normalize(VectorInput=Subtract(A=@n4.PS, B=@n5)), B=ScalarParameter() {defaultValue=3, parameterName=Face Depth})
  n7 = MaterialFunctionCall(
   UVs=ComponentMask(None=Add(A=@n4.PS, B=@n6)) {b=true, r=true}
   Texture Scale=VectorParameter() {parameterName=Face Scale, defaultValue=<rgba 20 20 0 0>}.RGB
  ) {materialFunction=&ScaleUVsByCenter}
  n8 = ComponentMask(None=VectorParameter() {parameterName=Face Position, defaultValue=<rgba 0.2476 -0.6506 0 0>}.RGB) {g=true, r=true}
  n9 = MaterialFunctionCall(Float2=MaterialFunctionCall(UVs=Add(A=@n7.UVs, B=@n8), "Rotation Angle (0-1)"=ScalarParameter() {defaultValue=0.2564, parameterName=Face Rotation}) {materialFunction=&CustomRotator}.Rotated Values) {materialFunction=&BreakOutFloat2Components}
  n10 = Transform(None=CameraVectorWS()) {transformSourceType=TRANSFORMSOURCE_World, transformType=TRANSFORM_Local}
  n11 = ComponentMask(None=@n10) {b=true}
  n12 = Add(A=@n9.R, B=Multiply(A=OneMinus(None=@n11), B=ScalarParameter() {defaultValue=0.4, parameterName=Face Reflection Z - Scale}))
  n13 = Add(
   A=@n9.G
   B=Multiply(A=ComponentMask(None=@n10) {g=true, r=true}, B=ScalarParameter() {defaultValue=0.01, parameterName=Face Reflection XY - Scale})
  )
  n14 = MaterialFunctionCall(
   "Animation  Phase (0-1)"=Frac(None=Multiply(A=@n2, B=@n3))
   Number of Rows=Constant() {r=1}
   Number of Columns=Constant() {r=2}
   Texture=TextureObject() {texture=&/T_BasketSuburb_Pickaxe_Flipbook_01}
   UVs=NamedRerouteDeclaration(None=MaterialFunctionCall(X=@n12, Y=@n13) {materialFunction=&MakeFloat2}.Result) {name=Camera UV, nodeColor=<rgba 1 0 0.4453 1>}
  ) {materialFunction=&FlipBook}
  n15 = ScalarParameter() {defaultValue=0.25, parameterName=Face Opacity}
  n16 = ScalarParameter() {defaultValue=0.1, parameterName=Face Opacity - F0}
  n17 = OneMinus(None=Fresnel(ExponentIn=ScalarParameter() {defaultValue=0.6149, parameterName=Face Opacity - Exp}) {baseReflectFraction=0})
  n18 = Add(
   A=NamedRerouteUsage()
   B=ComponentMask(None=VectorParameter() {parameterName=Face Mask Position, defaultValue=<rgba -0.4966 0.0626 0 0>}.RGB) {g=true, r=true}
  )
  n19 = ComponentMask(None=@n18) {r=true}
  n20 = ComponentMask(None=@n18) {g=true}
  n21 = Multiply(
   A=Step(Y=Abs(None=@n19), X=ScalarParameter() {defaultValue=0.412, parameterName=Face Mask X})
   B=Step(Y=Abs(None=@n20), X=ScalarParameter() {defaultValue=0.9378, parameterName=Face Mask Y})
  )
  n22 = OneMinus(None=NamedRerouteUsage())
  n23 = Multiply(A=Multiply(A=@n14.Alpha, B=LinearInterpolate(A=@n15, B=@n16, Alpha=@n17)), B=Saturate(None=Subtract(A=@n21, B=@n22)))
  FunctionOutput(None=SetMaterialAttributes!(MaterialAttributes=@n1.MaterialAttributes, EmissiveColor=LinearInterpolate(A=@n1.EmissiveColor, B=@n14.Result, Alpha=@n23)) {attributeSetTypes [#1]}) {outputName=SetMaterialAttributes}
  n24 = VectorParameter() {parameterName=Box Pos, defaultValue=<rgba 0.1878 0.25 0 0>}
  n25 = Subtract(
   A=TextureCoordinate()
   B=MaterialFunctionCall(X=@n24.R, Y=@n24.R) {materialFunction=&MakeFloat2}.Result
  )
  n26 = VectorParameter() {parameterName=Box Size, defaultValue=<rgba 0.0804 1.6727 0 0>}
  n27 = MaterialFunctionCall!(UVs=@n25, Size=@n26.RGB, CornerRoundness, KeepAspectRatio) {materialFunction=/FortUILibrary/MaterialFunctions/Shapes/SDFs/MF_SDF_Box}
  n28 = VectorParameter() {parameterName=Box Pos 2, defaultValue=<rgba 0.1734 0.3428 0 0>}
  n29 = MaterialFunctionCall(UVs=Subtract(A=@n25, B=ComponentMask(None=@n28.RGB) {g=true, r=true}), Size=Multiply(A=@n26.RGB) {constB=1.2}) {materialFunction=&MF_SDF_Box}
  NamedRerouteDeclaration(None=Max(A=Step(Y=@n27.SDF) {constX=0.001}, B=Step(Y=@n29.SDF) {constX=0.001})) {name=Blade Edge, nodeColor=<rgba 1 0 0.8906 1>}
  NamedRerouteDeclaration(None=ComponentMask(None=FunctionInput() {inputName=FX, inputType=FunctionInput_Vector4}) {g=true}) {name=Blade Mask, nodeColor=<rgba 0 0.25 1 1>}
  ReflectionVectorWS(CustomWorldNormal=Divide(A=LocalPosition().XYZ) {constB=100})
  ScalarParameter() {defaultValue=20, parameterName=Face Scale}
 ]}
